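# Niagara system for the mid-match sky island opening: a cracking rift effect with an opening mesh, forcefield, glows and floating fragments.
# System.Normalized_Phase drives growth across emitters, including stepped curves that advance crack materials in jolts, offset per crack layer.
# Flash, shockwave and energy waves add bursts, while a vertical ribbon pointer extends upward with time-of-day dimming.
# Some fragments use point attraction that flips from weak repulsion to strong pull late in its curve.
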NiagaraSystem /RiftPOICosmetics/Art/VFX/NS_MidMatch_SkyIsland_Opening
dependencies {
 usedDataInterfaces [&NiagaraDataInterfaceCurve, &NiagaraDataInterfaceVectorCurve, &NiagaraDataInterfaceColorCurve]
 usedDynamicInputs [&Divide_Float, &Subtract_Float]
 usedModules [
  &SystemState
  &self:SystemSpawnScript.SystemScriptSource.SystemScriptGraph.NiagaraNodeAssignment_0.SetVariables_#1
  &self:SystemSpawnScript.SystemScriptSource.SystemScriptGraph.NiagaraNodeAssignment_1.SetVariables_#2
  &Timeline
  &self:SystemSpawnScript.SystemScriptSource.SystemScriptGraph.NiagaraNodeAssignment_2.SetVariables_#3
  &self:SystemSpawnScript.SystemScriptSource.SystemScriptGraph.NiagaraNodeAssignment_3.SetVariables_#4
  &self:SystemSpawnScript.SystemScriptSource.SystemScriptGraph.NiagaraNodeAssignment_4.SetVariables_#5
 ]
}
emitters {
 Crack {
  emitterUpdateScript [
   &EmitterState(Life Cycle Mode=::NewEnumerator1, Loop Behavior=::NewEnumerator1, Loop Duration Mode=::NewEnumerator1, UseLoopDelay=false)
   &SpawnBurst_Instantaneous(Spawn Count=1, Use Spawn Probability=false)
  ]
  particleSpawnScript [
   &InitializeParticle(
    Lifetime=2
    Mesh Scale=&Multiply_VectorByFloat(Vector=<600 600 250>, Float=0.1725)
    Mesh Scale Mode=::NewEnumerator4
    UsePositionOffset=false
    Write Material Random=true
   )
   &InitialMeshOrientation(Mesh Axis To Orient=<0 0 1>, Mesh Orientation Mode=::NewEnumerator2, Use Rotation Vector=true)
  ]
  particleUpdateScript [
   &ParticleState(Kill Particles When Lifetime Has Elapsed=false, Let Infinitely Lived Particles Die When Emitter Deactivates=true)
   &DynamicMaterialParameters(
    Index 0 Param 1=&FloatFromCurve(
     FloatCurve=DataInterface() {propertyValues {
       OptimizeThreshold=0
       bOptimizeLUT=false
       bUseLUT=false
       Curve {keys [
         {arriveTangent=0.0238, interpMode=RCIM_Constant, time=0, value=0.01}
         <at 0.0033 0.0145 RCIM_Constant>
         <at 0.0101 0.017 RCIM_Constant>
         <at 0.0177 0.0204 RCIM_Constant>
         <at 0.026 0.0245 RCIM_Constant>
         <at 0.2163 0.0259 RCIM_Constant>
         <at 0.2221 0.0277 RCIM_Constant>
         <at 0.2313 0.0308 RCIM_Constant>
         <at 0.2467 0.035 RCIM_Constant>
         <at 0.4177 0.0348 RCIM_Constant>
         <at 0.4227 0.0372 RCIM_Constant>
         <at 0.4283 0.0394 RCIM_Constant>
         <at 0.4324 0.0426 RCIM_Constant>
         <at 0.448 0.0428 RCIM_Constant>
         <at 0.452 0.0467 RCIM_Constant>
         <at 0.4565 0.0509 RCIM_Constant>
         <at 0.6339 0.0509 RCIM_Constant>
         <at 0.6564 0.0555 RCIM_Constant>
         <at 0.6568 0.0604 RCIM_Constant>
         <at 0.6588 0.065 RCIM_Constant>
         <at 0.7848 0.065 RCIM_Constant>
         <at 0.7881 0.0685 RCIM_Constant>
         <at 0.7916 0.0727 RCIM_Constant>
         <at 0.8153 0.0728 RCIM_Constant>
         <at 0.8182 0.0755 RCIM_Constant>
         <at 0.8221 0.0778 RCIM_Constant>
         <at 0.8249 0.0803 RCIM_Constant>
         <at 0.8518 0.0802 RCIM_Constant>
         <at 0.855 0.0835 RCIM_Constant>
         <at 0.8571 0.0869 RCIM_Constant>
         <at 0.863 0.0885 RCIM_Constant>
         <at 0.8651 0.09 RCIM_Constant>
         <at 0.9851 0.0901 RCIM_Constant>
         <at 0.9918 0.0934 RCIM_Constant>
         <at 0.9955 0.0975 RCIM_Constant>
         <at 1 0.1 RCIM_Constant>
        ]}
      }}
     CurveIndex=$System.Normalized_Phase
     Scale Curve=$System.ScaleCracks
    )
    Index 0 Param 2=&Lerp_Float(A=5, B=10, Alpha=$System.Normalized_Phase)
    Index 0 Param 3=100
    Module.Param0WriteEnabled=true
    Module.Param1WriteEnabled=true
    Module.Param2WriteEnabled=true
    Module.Param3WriteEnabled=true
   )
   &CameraOffset(Camera Offset Amount=200)
  ]
  rendererClasses [&NiagaraMeshRendererProperties]
  renderers [{rendererClass=&NiagaraMeshRendererProperties, rendererIndex=0}]
 }
 CrackSmall001 {
  emitterUpdateScript [
   &EmitterState(Life Cycle Mode=::NewEnumerator1, Loop Behavior=::NewEnumerator1, Loop Duration Mode=::NewEnumerator1, UseLoopDelay=false)
   &SpawnBurst_Instantaneous(Spawn Count=1, Use Spawn Probability=false)
  ]
  particleSpawnScript [
   &InitializeParticle(
    Lifetime=2
    Mesh Scale=&Multiply_VectorByFloat(Vector=<600 600 250>, Float=0.1725)
    Mesh Scale Mode=::NewEnumerator4
    UsePositionOffset=false
    Write Material Random=true
   )
   &InitialMeshOrientation(Mesh Axis To Orient=<0 0 1>, Mesh Orientation Mode=::NewEnumerator2, Rotation=<0.55 0 0>, Use Rotation Vector=true)
  ]
  particleUpdateScript [
   &ParticleState(Kill Particles When Lifetime Has Elapsed=false, Let Infinitely Lived Particles Die When Emitter Deactivates=true)
   &DynamicMaterialParameters(
    Index 0 Param 1=&FloatFromCurve(
     FloatCurve=DataInterface() {propertyValues {
       bOptimizeLUT=false
       bUseLUT=false
       Curve {keys [<at 0 0.01 RCIM_Constant>,<at 0.0033 0.0145 RCIM_Constant>,<at 0.0101 0.017 RCIM_Constant>,<at 0.0177 0.0204 RCIM_Constant>,<at 0.026 0.0245 RCIM_Constant>,<at 0.2163 0.0259 RCIM_Constant>,<at 0.2221 0.0277 RCIM_Constant>,<at 0.2313 0.0308 RCIM_Constant>,<at 0.2467 0.035 RCIM_Constant>,<at 0.4177 0.0348 RCIM_Constant>,<at 0.4227 0.0372 RCIM_Constant>,<at 0.4283 0.0394 RCIM_Constant>,<at 0.4324 0.0426 RCIM_Constant>,<at 0.448 0.0428 RCIM_Constant>,<at 0.452 0.0467 RCIM_Constant>,<at 0.4565 0.0509 RCIM_Constant>,<at 0.6339 0.0509 RCIM_Constant>,<at 0.6564 0.0555 RCIM_Constant>,<at 0.6568 0.0604 RCIM_Constant>,<at 0.6588 0.065 RCIM_Constant>,<at 0.7848 0.065 RCIM_Constant>,<at 0.7881 0.0685 RCIM_Constant>,<at 0.7916 0.0727 RCIM_Constant>,<at 0.8153 0.0728 RCIM_Constant>,<at 0.8182 0.0755 RCIM_Constant>,<at 0.8221 0.0778 RCIM_Constant>,<at 0.8249 0.0803 RCIM_Constant>,<at 0.8518 0.0802 RCIM_Constant>,<at 0.855 0.0835 RCIM_Constant>,<at 0.8571 0.0869 RCIM_Constant>,<at 0.863 0.0885 RCIM_Constant>,<at 0.8651 0.09 RCIM_Constant>,<at 0.9851 0.0901 RCIM_Constant>,<at 0.9918 0.0934 RCIM_Constant>,<at 0.9955 0.0975 RCIM_Constant>,<at 1 0.1 RCIM_Constant>]}
      }}
     CurveIndex=&Add_Float(A=$System.Normalized_Phase, B=-0.05)
     Scale Curve=$System.ScaleCracks
    )
    Index 0 Param 2=&Lerp_Float(A=5, B=10, Alpha=$System.Normalized_Phase)
    Index 0 Param 3=1
    Module.Param0WriteEnabled=true
    Module.Param1WriteEnabled=true
    Module.Param2WriteEnabled=true
    Module.Param3WriteEnabled=true
   )
   &CameraOffset(Camera Offset Amount=200)
  ]
  rendererClasses [&NiagaraMeshRendererProperties]
  renderers [{rendererClass=&NiagaraMeshRendererProperties, rendererIndex=0}]
 }
 EnergyWave001 {
  emitterUpdateScript [
   &EmitterState(
    Delay First Loop Only=true
    Life Cycle Mode=::NewEnumerator1
    Loop Delay=0.2
    Loop Duration=&RandomRangeFloat(Minimum=0.4, Maximum=0.6)
    UseLoopDelay=true
   )
   &SpawnBurst_Instantaneous(Spawn Count=1, Spawn Probability=HlslExpression() {hlslExpression="Emitter.LoopCount < 1"}, Use Spawn Probability=false)
  ]
  particleSpawnScript [
   &InitializeParticle(
    Lifetime=2
    Mesh Scale Mode=::NewEnumerator3
    Mesh Uniform Scale=&Lerp_Float(A=1, B=25, Alpha=$System.Normalized_Phase)
    UsePositionOffset=false
    Write Material Random=true
   )
   &CameraOffset(Camera Offset Amount=200)
   &InitialMeshOrientation(Mesh Axis To Orient=<0 0 1>, Mesh Orientation Mode=::NewEnumerator2, Use Rotation Vector=false)
  ]
  particleUpdateScript [
   &ParticleState()
   &DynamicMaterialParameters(
    Index 0 Param 1=&FloatFromCurve(FloatCurve=DataInterface() {propertyValues {Curve {keys [
         {arriveTangent=3.3579, interpMode=RCIM_Cubic, leaveTangent=3.3579, tangentMode=RCTM_User, time=0, value=-0.3}
         {arriveTangent=0.0871, interpMode=RCIM_Cubic, leaveTangent=0.0871, tangentMode=RCTM_User, time=1, value=1}
        ]}}})
    Index 0 Param 2=&FloatFromCurve(FloatCurve=DataInterface() {propertyValues {Curve {keys [{arriveTangent=1.5982, interpMode=RCIM_Cubic, leaveTangent=1.5982, tangentMode=RCTM_User, time=0, value=0.5}, <at 1 1>]}}})
    Index 0 Param 3=&FloatFromCurve(FloatCurve=DataInterface() {propertyValues {Curve {keys [<at 0 0>,<at 0.8 1>,<at 1 0>]}}})
    Module.Param0WriteEnabled=true
    Module.Param1WriteEnabled=true
    Module.Param2WriteEnabled=true
    Module.Param3WriteEnabled=false
   )
   &ScaleMeshSize(Scale Factor=&VectorFromFloat(Value=&Multiply_Float(
      A=&FloatFromCurve(FloatCurve=DataInterface() {propertyValues {Curve {keys [
           {interpMode=RCIM_Cubic, tangentMode=RCTM_User, time=0, value=1.5}
           {arriveTangent=-1.6082, interpMode=RCIM_Cubic, leaveTangent=-1.6082, tangentMode=RCTM_User, time=1, value=0.8}
          ]}}})
      B=1
     )))
   &ScaleColor(
    Scale Alpha=&FloatFromCurve(
     FloatCurve=DataInterface() {propertyValues {Curve {keys [
         {interpMode=RCIM_Cubic, tangentMode=RCTM_User, time=0, value=0}
         {interpMode=RCIM_Cubic, tangentMode=RCTM_User, time=0.2, value=1}
         {tangentMode=RCTM_User, time=1, value=0}
        ]}}}
     Scale Curve=0.5
    )
    Scale RGB=&MakeVectorFromLinearColorRGB(Color=&ColorFromCurve(ColorCurve=DataInterface() {propertyValues {
        AlphaCurve {keys [<at 0 1 RCIM_Cubic>]}
        BlueCurve {keys [<at 0 1 RCIM_Cubic>,<at 1 1 RCIM_Cubic>]}
        GreenCurve {keys [<at 0 0.3261 RCIM_Cubic>,<at 1 0 RCIM_Cubic>]}
        RedCurve {keys [<at 0 0 RCIM_Cubic>,<at 1 0.7641 RCIM_Cubic>]}
       }}))
    ScaleA=true
    ScaleRGB=true
   )
   &ScaleColor(
    Scale Alpha=&FloatFromCurve(FloatCurve=DataInterface() {propertyValues {Curve {keys [<at 0 1>,<at 1 0>]}}}, CurveIndex=$System.Timeline.NormalizedTime)
    ScaleA=true
    ScaleRGB=true
   ) {module=ScaleColor001}
  ]
  rendererClasses [&NiagaraMeshRendererProperties]
  renderers [{rendererClass=&NiagaraMeshRendererProperties, rendererIndex=0}]
 }
 FlareLightShockwave001 {
  emitterSpawnScript [/SetVariables(Emitter.DepthFade=100) {isSetParameters=true}]
  emitterUpdateScript [
   &EmitterState(Life Cycle Mode=::NewEnumerator1, Loop Behavior=::NewEnumerator1, Loop Duration=1, UseLoopDelay=false)
   &SpawnBurst_Instantaneous(Use Spawn Probability=false, UseLoopCountLimit=false)
  ]
  particleSpawnScript [
   &InitializeParticle(
    Color=<rgba 0.1598 0.398 1 1>
    Lifetime=0.8
    Mesh Scale Mode=::NewEnumerator3
    Position Offset=<15 0 0>
    Sprite Size Mode=::NewEnumerator3
    Uniform Sprite Size=20000
    UsePositionOffset=true
    Write Material Random=true
   )
   &ShapeLocation(MaskSpawnGroup=false, Offset=$System.Offset, Offset Mode=::NewEnumerator0, Override Seed=false, Sphere Radius=0, UseOwnerScale=true)
  ]
  particleUpdateScript [
   &ParticleState()
   &CameraOffset(Camera Offset Amount=&FloatFromCurve(
     FloatCurve=DataInterface() {propertyValues {Curve {keys [
         {arriveTangent=4.0899, interpMode=RCIM_Cubic, leaveTangent=4.0899, tangentMode=RCTM_User, time=0, value=0.5}
         {arriveTangent=-3.6426, interpMode=RCIM_Cubic, leaveTangent=-3.6426, tangentMode=RCTM_User, time=0.2, value=1}
         {arriveTangent=5.165, interpMode=RCIM_Cubic, leaveTangent=5.165, tangentMode=RCTM_User, time=0.4, value=0.3}
         <at 0.75 0>
        ]}}}
     Scale Curve=200
    ))
   &ScaleMeshSize(Scale Factor=&VectorFromFloat(Value=&FloatFromCurve(
      FloatCurve=DataInterface() {propertyValues {Curve {keys [
          {arriveTangent=0.8167, interpMode=RCIM_Cubic, leaveTangent=0.8167, tangentMode=RCTM_User, time=0, value=0.1}
          {arriveTangent=0.3412, interpMode=RCIM_Cubic, leaveTangent=0.3412, tangentMode=RCTM_User, time=0.2479, value=0.2161}
          {
           arriveTangent=2.4818
           arriveTangentWeight=0.1302
           interpMode=RCIM_Cubic
           leaveTangent=2.4818
           leaveTangentWeight=0.2603
           tangentMode=RCTM_User
           time=0.271
           value=0.245
          }
          {arriveTangent=0.3013, interpMode=RCIM_Cubic, leaveTangent=0.3013, tangentMode=RCTM_User, time=0.75, value=1}
         ]}}}
      CurveIndex=$Particles.Age
      Scale Curve=2
     )))
   &DynamicMaterialParameters(
    Index 0 Param 1=1
    Index 0 Param 2=&Lerp_Float(
     A=1
     B=0.3
     Alpha=&FloatFromCurve(FloatCurve=DataInterface() {propertyValues {Curve {keys [
          {arriveTangent=5.1433, interpMode=RCIM_Cubic, time=0, value=0}
          {arriveTangent=4.5726, interpMode=RCIM_Cubic, leaveTangent=4.5726, tangentMode=RCTM_User, time=0.25, value=0.7}
          {arriveTangent=-6.4921, interpMode=RCIM_Cubic, leaveTangent=2.9849, tangentMode=RCTM_Break, time=0.4, value=0.35}
          {arriveTangent=0.6574, interpMode=RCIM_Cubic, leaveTangent=0.6574, tangentMode=RCTM_User, time=1, value=1}
         ]}}})
    )
    Index 3 Param 1=&Multiply_Float(A=&MakeFloatFromVector(Vector=$Particles.Scale, Channel=::NewEnumerator0), B=90)
    Index 3 Param 2=&FloatFromCurve(FloatCurve=DataInterface() {propertyValues {Curve {keys [
         {arriveTangent=2.1437, interpMode=RCIM_Cubic, leaveTangent=2.1437, tangentMode=RCTM_User, time=0, value=0}
         {arriveTangent=1.3259, interpMode=RCIM_Cubic, leaveTangent=1.3259, tangentMode=RCTM_User, time=0.6, value=0.5}
         {interpMode=RCIM_Cubic, leaveTangent=0.2859, time=1, value=1}
        ]}}})
    Module.Param0WriteEnabled=true
    Module.Param0WriteEnabled3=true
    Module.Param1WriteEnabled=true
    Module.Param1WriteEnabled3=true
    Module.Param2WriteEnabled=false
    Module.Param2WriteEnabled3=false
    Module.Param3WriteEnabled=false
    Module.Param3WriteEnabled3=false
    Write Parameter Index 3=true
   )
   &ScaleSpriteSize(Uniform Curve Sprite Scale=DataInterface() {propertyValues {Curve {keys [
        {arriveTangent=-6.1148, interpMode=RCIM_Cubic, leaveTangent=-6.1148, tangentMode=RCTM_User, time=0, value=1}
        {arriveTangent=-0.4947, interpMode=RCIM_Cubic, leaveTangent=8.7488, tangentMode=RCTM_Break, time=0.4, value=0.3}
        {
         arriveTangent=-0.2276
         arriveTangentWeight=0.0565
         interpMode=RCIM_Cubic
         leaveTangent=-0.2276
         leaveTangentWeight=0.1797
         tangentMode=RCTM_User
         time=0.5
         value=0.9
        }
        {interpMode=RCIM_Cubic, leaveTangent=-0.9894, time=0.7, value=0.5}
       ]}}})
   &ScaleColor(
    Scale Alpha=&FloatFromCurve(FloatCurve=DataInterface() {propertyValues {Curve {keys [
         {arriveTangent=18.2404, interpMode=RCIM_Cubic, leaveTangent=18.2404, tangentMode=RCTM_User, time=0, value=1}
         {arriveTangent=-1.4815, interpMode=RCIM_Cubic, leaveTangent=-1.4815, time=0.12, value=2}
         {
          arriveTangent=-4.2653
          arriveTangentWeight=0.0565
          interpMode=RCIM_Cubic
          leaveTangent=-4.2653
          leaveTangentWeight=0.1797
          tangentMode=RCTM_User
          time=0.27
          value=0.6
         }
         {arriveTangent=-11.3751, interpMode=RCIM_Cubic, leaveTangent=-11.3751, tangentMode=RCTM_User, time=0.44, value=2}
         {interpMode=RCIM_Cubic, tangentMode=RCTM_User, time=1, value=0}
        ]}}})
    ScaleA=true
    ScaleRGB=true
   )
   /SetVariables(Particles.WaveColor:LinearColor=&LerpLinearColors(
     StartColor=<rgba 0.7113 0.6033 1 0>
     EndColor=<rgba 0.5788 1.0477 1.25 1>
     LerpFactor=&FloatFromCurve(FloatCurve=DataInterface() {propertyValues {Curve {keys [
          {arriveTangent=18.2701, interpMode=RCIM_Cubic, time=0, value=0}
          <at 0.15 1>
          {arriveTangent=29.7321, interpMode=RCIM_Cubic, leaveTangent=29.7321, tangentMode=RCTM_User, time=0.27, value=0.5}
          {arriveTangentWeight=0.0997, interpMode=RCIM_Cubic, leaveTangentWeight=0.3488, tangentMode=RCTM_User, time=0.3, value=1.1}
          {interpMode=RCIM_Cubic, leaveTangent=-2.4954, time=1, value=0}
         ]}}})
    )) {isSetParameters=true}
   /SetVariables(Particles.LightColor=&MultiplyLinearColor(Linear Color=$Particles.Color, Scale Factor=&MakeVector4(X=1, Y=1, Z=1, W=15))) {isSetParameters=true}
   /SetVariables(Particles.LightRadius=&FloatFromCurve(FloatCurve=DataInterface() {propertyValues {Curve {keys [<at 0.2 0.35 RCIM_Cubic>,<at 0.5 1>]}}}, Scale Curve=10000)) {isSetParameters=true}
  ]
  rendererClasses [
   &NiagaraSpriteRendererProperties
   &NiagaraLightRendererProperties
   &NiagaraMeshRendererProperties
  ]
  renderers [
   {rendererClass=&NiagaraSpriteRendererProperties}
   {rendererClass=&NiagaraLightRendererProperties, rendererIndex=1}
   {rendererClass=&NiagaraMeshRendererProperties, rendererIndex=2}
  ]
 }
 Flash001 {
  emitterUpdateScript [
   &EmitterState(Life Cycle Mode=::NewEnumerator1, Loop Behavior=::NewEnumerator1, Loop Duration=1, UseLoopDelay=false)
   &SpawnBurst_Instantaneous(Use Spawn Probability=false, UseLoopCountLimit=false)
  ]
  particleSpawnScript [
   &InitializeParticle(
    Color=<rgba 0.5207 0.6194 1 1>
    Lifetime=0.5
    Sprite Size Mode=::NewEnumerator3
    Uniform Sprite Size=10000
    UsePositionOffset=false
    Write Material Random=true
   )
   &ShapeLocation(MaskSpawnGroup=false, Offset=$System.Offset, Offset Mode=::NewEnumerator0, Override Seed=false, Sphere Radius=0, UseOwnerScale=true)
   &DynamicMaterialParameters(
    Index 1 Param 1=150
    Module.Param0WriteEnabled=true
    Module.Param0WriteEnabled1=true
    Module.Param1WriteEnabled=true
    Module.Param1WriteEnabled1=true
    Module.Param2WriteEnabled=true
    Module.Param2WriteEnabled1=true
    Module.Param3WriteEnabled=true
    Module.Param3WriteEnabled1=true
    Write Parameter Index 1=true
   )
  ]
  particleUpdateScript [
   &ParticleState()
   &CameraOffset(Camera Offset Amount=&FloatFromCurve(
     FloatCurve=DataInterface() {propertyValues {Curve {keys [
         {arriveTangent=4.0899, interpMode=RCIM_Cubic, leaveTangent=4.0899, tangentMode=RCTM_User, time=0, value=0.5}
         {arriveTangent=-3.6426, interpMode=RCIM_Cubic, leaveTangent=-3.6426, tangentMode=RCTM_User, time=0.2, value=1}
         {arriveTangent=5.165, interpMode=RCIM_Cubic, leaveTangent=5.165, tangentMode=RCTM_User, time=0.4, value=0.3}
         <at 0.75 0>
        ]}}}
     Scale Curve=200
    ))
   &ScaleSpriteSize(Uniform Curve Sprite Scale=DataInterface() {propertyValues {Curve {keys [
        {interpMode=RCIM_Cubic, leaveTangentWeight=0.5401, tangentMode=RCTM_User, time=0, value=0.1}
        {arriveTangent=2.0921, interpMode=RCIM_Cubic, leaveTangent=2.0921, tangentMode=RCTM_User, time=0.1, value=0.2}
        {interpMode=RCIM_Cubic, tangentMode=RCTM_User, time=0.75, value=1}
        {interpMode=RCIM_Cubic, tangentMode=RCTM_User, time=1, value=0.4}
       ]}}})
   &ScaleColor(
    Scale Alpha=&FloatFromCurve(FloatCurve=DataInterface() {propertyValues {Curve {keys [
         {arriveTangent=2.8296, interpMode=RCIM_Cubic, leaveTangent=2.8296, tangentMode=RCTM_User, time=0, value=0.4}
         {interpMode=RCIM_Cubic, tangentMode=RCTM_User, time=0.7, value=1}
         {arriveTangent=-8.4365, interpMode=RCIM_Cubic, leaveTangent=-8.4365, tangentMode=RCTM_User, time=1, value=0}
        ]}}})
    ScaleA=true
    ScaleRGB=true
   )
  ]
  rendererClasses [&NiagaraSpriteRendererProperties]
  renderers [{rendererClass=&NiagaraSpriteRendererProperties}]
 }
 Forcefield {
  emitterUpdateScript [
   &EmitterState(Life Cycle Mode=::NewEnumerator1, Loop Behavior=::NewEnumerator1, Loop Duration Mode=::NewEnumerator1, UseLoopDelay=false)
   &SpawnBurst_Instantaneous(Spawn Count=1, Use Spawn Probability=false)
  ]
  particleSpawnScript [
   /Niagara/Modules/Spawn/Initialization/InitializeParticle(
    Color=<rgba 0 0.3261 1 1>
    Lifetime=5
    Mesh Scale=&VectorFromFloat(Value=100)
    Module.Write Color=true
    Module.Write Lifetime=true
    Module.Write Mass=true
    Module.Write Material Random=true
    Module.Write Position=false
    Module.Write Scale=true
    Module.Write SpriteRotation=false
    Module.Write SpriteSize=false
    Position=<0 0 4000>
   )
   &InitialMeshOrientation(Mesh Axis To Orient=<0 0 -1>, Mesh Orientation Mode=::NewEnumerator2, Use Rotation Vector=false)
  ]
  particleUpdateScript [
   &ParticleState(Kill Particles When Lifetime Has Elapsed=false, Let Infinitely Lived Particles Die When Emitter Deactivates=true)
   &ScaleMeshSize(Scale Factor=&VectorFromFloat(Value=&Lerp_Float(A=0.1, B=1, Alpha=$System.Normalized_Phase)))
   &DynamicMaterialParameters(
    Index 0 Param 1=&FloatFromCurve(FloatCurve=DataInterface() {propertyValues {Curve {keys [{arriveTangent=5.3858, interpMode=RCIM_Cubic, time=0, value=0}, <at 0.974 1 RCIM_Cubic>]}}})
    Index 0 Param 2=2
    Index 0 Param 3=0.4
    Index 0 Param 4=&FloatFromCurve(FloatCurve=DataInterface() {propertyValues {Curve {keys [<at 0 0.5>,<at 1 12>]}}})
    Module.Param0WriteEnabled=true
    Module.Param1WriteEnabled=true
    Module.Param2WriteEnabled=true
    Module.Param3WriteEnabled=true
   )
   &ScaleColor(
    Scale RGB=&VectorFromCurve(VectorCurve=DataInterface() {propertyValues {XCurve {keys [<at 0 2>,<at 1 0>]}, YCurve {keys [<at 0 2>,<at 1 1>]}, ZCurve {keys [<at 0 2>,<at 1 0.25>]}}})
    ScaleA=true
    ScaleRGB=false
   )
  ]
  rendererClasses [&NiagaraMeshRendererProperties]
  renderers [{rendererClass=&NiagaraMeshRendererProperties, rendererIndex=0}]
 }
 Fragments {
  emitterUpdateScript [
   &EmitterState(Life Cycle Mode=::NewEnumerator1, Loop Behavior=::NewEnumerator1, Loop Duration Mode=::NewEnumerator1, UseLoopDelay=false)
   &SpawnBurst_Instantaneous(
    Spawn Count=20
    Spawn Probability=&MakeCustomFloatFromBool(Bool=&SetBoolByFloatComparison(A=$System.Normalized_Phase, B=0.5), True Float=1, False Float=0)
    Use Spawn Probability=true
    UseLoopCountLimit=false
   )
  ]
  particleSpawnScript [
   /SetVariables(Particles.RandCurve:NiagaraFloat=&FloatFromCurve(
     FloatCurve=DataInterface() {propertyValues {Curve {keys [
         {interpMode=RCIM_Cubic, tangentMode=RCTM_User, time=0, value=0}
         {arriveTangent=0.6613, interpMode=RCIM_Cubic, leaveTangent=0.6613, tangentMode=RCTM_User, time=0.7, value=0.2}
         {arriveTangentWeight=0.4215, interpMode=RCIM_Cubic, tangentMode=RCTM_User, time=1, value=1}
        ]}}}
     CurveIndex=&RandomRangeFloat(Minimum=0, Maximum=1)
    )) {isSetParameters=true}
   /SetVariables(Particles.MeshRand:NiagaraInt32=&RandomRangeInt(Minimum=0, Maximum=2)) {isSetParameters=true, module=SetVariables_#6}
   /Niagara/Modules/Spawn/Initialization/V2/InitializeParticle(
    Color Mode=::NewEnumerator0
    Lifetime=5
    Mesh Scale Mode=::NewEnumerator3
    Mesh Uniform Scale=&Lerp_Float(A=45, B=100, Alpha=$Particles.RandCurve)
    UsePositionOffset=false
    Write Material Random=true
   )
   &ShapeLocation(
    Cylinder Height=250
    Cylinder Radius=&Lerp_Float(A=2500, B=5000, Alpha=$Particles.RandCurve)
    MaskSpawnGroup=false
    Offset=<0 0 -1000>
    Offset Mode=::NewEnumerator0
    Override Seed=false
    Shape Primitive=::NewEnumerator1
    Surface Only=false
    UseOwnerScale=true
   )
   &InitialMeshOrientation(
    Mesh Orientation Mode=::NewEnumerator2
    Orientation Vector=<0 0 1>
    Rotation=&RandomRangeVector(Minimum=<0 0 0>, Maximum=<0 0.25 1>)
    Use Rotation Vector=true
   )
  ]
  particleUpdateScript [
   &ParticleState(Kill Particles When Lifetime Has Elapsed=false, Let Infinitely Lived Particles Die When Emitter Deactivates=true)
   &UpdateMeshOrientation(
    Rotation Rate=&RandomRangeFloat(Minimum=0.01, Maximum=0.02)
    Rotation Vector=&RandomRangeVector(Minimum=<0 0 0>, Maximum=<1 1 1>)
   )
  ]
  rendererClasses [&NiagaraMeshRendererProperties]
  renderers [{rendererClass=&NiagaraMeshRendererProperties, rendererIndex=0}]
 }
 Fragments001 {
  emitterUpdateScript [
   &EmitterState(Life Cycle Mode=::NewEnumerator1, Loop Behavior=::NewEnumerator1, Loop Duration Mode=::NewEnumerator1, UseLoopDelay=false)
   &SpawnRate(SpawnRate=&Lerp_Float(A=5, B=10, Alpha=$System.Normalized_Phase), Use Spawn Probability=false)
  ]
  particleSpawnScript [
   /SetVariables(Particles.MeshRand:NiagaraInt32=&RandomRangeInt(Minimum=0, Maximum=2)) {isSetParameters=true, module=SetVariables_#6}
   &InitializeParticle(
    Color Mode=::NewEnumerator0
    Lifetime=10
    Mesh Scale Mode=::NewEnumerator3
    Mesh Uniform Scale=&Lerp_Float(
     A=&RandomRangeFloat(Minimum=10, Maximum=20)
     B=&RandomRangeFloat(Minimum=20, Maximum=40)
     Alpha=$System.Normalized_Phase
    )
    UsePositionOffset=false
    Write Material Random=true
   )
   &ShapeLocation(
    MaskSpawnGroup=false
    Override Seed=false
    Sphere Radius=&Lerp_Float(A=380, B=3800, Alpha=$System.Normalized_Phase)
    Sphere Surface Distribution=0.9
    UseOwnerScale=true
   )
   &InitialMeshOrientation(
    Mesh Orientation Mode=::NewEnumerator2
    Orientation Vector=<0 0 1>
    Rotation=&RandomRangeVector(Minimum=<0 0 0>, Maximum=<0 0.25 1>)
    Use Rotation Vector=true
   )
  ]
  particleUpdateScript [
   &ParticleState()
   &ScaleMeshSize(Scale Factor=&VectorFromFloat(Value=&FloatFromCurve(FloatCurve=DataInterface() {propertyValues {Curve {keys [<at 0 0>,<at 0.2 1>,<at 0.8 1>,<at 1 0>]}}})))
   &UpdateMeshOrientation(
    Rotation Rate=&FloatFromCurve(
     FloatCurve=DataInterface() {propertyValues {Curve {keys [
         {interpMode=RCIM_Cubic, tangentMode=RCTM_User, time=0.2, value=0.1}
         {arriveTangent=3.6747, interpMode=RCIM_Cubic, leaveTangent=3.6747, tangentMode=RCTM_User, time=1, value=1}
        ]}}}
     Scale Curve=&Multiply_Float(
      A=&RandomRangeFloat(Minimum=0.1, Maximum=0.4)
      B=&MakeCustomFloatFromBool(Bool=true, True Float=1, False Float=-1)
     )
    )
    Rotation Vector=&RandomRangeVector(Minimum=<0 0 0>, Maximum=<1 1 1>)
   )
   &PointAttractionForce(
    Attraction Radius=10000
    Attraction Strength=&FloatFromCurve(
     FloatCurve=DataInterface() {propertyValues {Curve {keys [
         {interpMode=RCIM_Cubic, tangentMode=RCTM_User, time=0, value=0.1}
         {interpMode=RCIM_Cubic, tangentMode=RCTM_User, time=0.9, value=0.3}
         {arriveTangentWeight=0.0756, interpMode=RCIM_Cubic, tangentMode=RCTM_User, time=1, value=1}
        ]}}}
     Scale Curve=&Lerp_Float(
      A=-7
      B=800
      Alpha=&FloatFromCurve(FloatCurve=DataInterface() {propertyValues {Curve {keys [<at 0 0>,<at 0.95 0>,<at 0.99 1>]}}})
     )
    )
    Falloff Exponent=0.08
    Kill Within Radius=false
    Module.Use Falloff=true
    UseAttractorOffset=false
   )
   &SolveForcesAndVelocity(Clamp Velocity=false, Limit Acceleration=false)
  ]
  rendererClasses [&NiagaraMeshRendererProperties]
  renderers [{rendererClass=&NiagaraMeshRendererProperties, rendererIndex=0}]
 }
 Fragments002 {
  emitterUpdateScript [
   &EmitterState(Life Cycle Mode=::NewEnumerator1, Loop Behavior=::NewEnumerator1, Loop Duration Mode=::NewEnumerator1, UseLoopDelay=false)
   &SpawnBurst_Instantaneous(Spawn Count=20, Use Spawn Probability=false, UseLoopCountLimit=false)
  ]
  particleSpawnScript [
   /SetVariables(Particles.MeshRand:NiagaraInt32=&RandomRangeInt(Minimum=0, Maximum=2)) {isSetParameters=true, module=SetVariables_#6}
   &InitializeParticle(
    Color Mode=::NewEnumerator0
    Lifetime=10
    Mesh Scale Mode=::NewEnumerator3
    Mesh Uniform Scale=&Lerp_Float(
     A=&RandomRangeFloat(Minimum=10, Maximum=20)
     B=&RandomRangeFloat(Minimum=20, Maximum=40)
     Alpha=$System.Normalized_Phase
    )
    UsePositionOffset=false
    Write Material Random=true
   )
   &ShapeLocation(MaskSpawnGroup=false, Override Seed=false, Sphere Surface Distribution=0.9, UseOwnerScale=true)
   &InitialMeshOrientation(
    Mesh Orientation Mode=::NewEnumerator2
    Orientation Vector=<0 0 1>
    Rotation=&RandomRangeVector(Minimum=<0 0 0>, Maximum=<0 0.25 1>)
    Use Rotation Vector=true
   )
   &AddVelocity(Velocity=&Multiply_VectorByFloat(Vector=$Particles.ShapeLocation.ShapeNormal, Float=&RandomRangeFloat(Minimum=2000, Maximum=4000)))
  ]
  particleUpdateScript [
   &ParticleState()
   &ScaleMeshSize(Scale Factor=&VectorFromFloat(Value=&FloatFromCurve(FloatCurve=DataInterface() {propertyValues {Curve {keys [<at 0 0>,<at 0.2 1>,<at 0.8 1>,<at 1 0>]}}})))
   &UpdateMeshOrientation(
    Rotation Rate=&FloatFromCurve(
     FloatCurve=DataInterface() {propertyValues {Curve {keys [
         {interpMode=RCIM_Cubic, tangentMode=RCTM_User, time=0.2, value=0.1}
         {arriveTangent=3.6747, interpMode=RCIM_Cubic, leaveTangent=3.6747, tangentMode=RCTM_User, time=1, value=1}
        ]}}}
     Scale Curve=&Multiply_Float(
      A=&RandomRangeFloat(Minimum=0.1, Maximum=0.4)
      B=&MakeCustomFloatFromBool(Bool=true, True Float=1, False Float=-1)
     )
    )
    Rotation Vector=&RandomRangeVector(Minimum=<0 0 0>, Maximum=<1 1 1>)
   )
   &PointAttractionForce(
    Attraction Radius=10000
    Attraction Strength=&FloatFromCurve(
     FloatCurve=DataInterface() {propertyValues {Curve {keys [
         {interpMode=RCIM_Cubic, tangentMode=RCTM_User, time=0, value=0.1}
         {interpMode=RCIM_Cubic, tangentMode=RCTM_User, time=0.9, value=0.3}
         {arriveTangentWeight=0.0756, interpMode=RCIM_Cubic, tangentMode=RCTM_User, time=1, value=1}
        ]}}}
     Scale Curve=&Lerp_Float(
      A=-7
      B=800
      Alpha=&FloatFromCurve(FloatCurve=DataInterface() {propertyValues {Curve {keys [<at 0 0>,<at 0.95 0>,<at 0.99 1>]}}})
     )
    )
    Falloff Exponent=0.08
    Kill Within Radius=false
    Module.Use Falloff=true
    UseAttractorOffset=false
   )
   &Drag(Drag=2, Use Linear Drag=true, Use Rotational Drag=true) {module=Drag001}
   &SolveForcesAndVelocity(Clamp Velocity=false, Limit Acceleration=false)
  ]
  rendererClasses [&NiagaraMeshRendererProperties]
  renderers [{rendererClass=&NiagaraMeshRendererProperties, rendererIndex=0}]
 }
 Glow {
  emitterUpdateScript [
   &EmitterState(Life Cycle Mode=::NewEnumerator1, Loop Behavior=::NewEnumerator1, Loop Delay=0.2, Loop Duration Mode=::NewEnumerator1, UseLoopDelay=true)
   &SpawnBurst_Instantaneous(Spawn Count=3, Use Spawn Probability=false, UseLoopCountLimit=false)
   &SpawnRate(
    Spawn Probability=&MakeCustomFloatFromBool(
     Bool=&SetBoolByFloatComparison(A=$System.Timeline.NormalizedTime, B=1, Comparison Type=::NewEnumerator4)
     True Float=0
     False Float=1
    )
    SpawnRate=20
    Use Spawn Probability=true
   )
  ]
  particleSpawnScript [
   /Niagara/Modules/Spawn/Initialization/InitializeParticle(
    Color=<rgba 0 0.6521 2 1>
    Lifetime=0.2
    Module.Write Color=true
    Module.Write Lifetime=true
    Module.Write Mass=false
    Module.Write Material Random=true
    Module.Write Position=true
    Module.Write Scale=false
    Module.Write SpriteRotation=true
    Module.Write SpriteSize=true
    Sprite Rotation=&RandomRangeFloat(Minimum=0, Maximum=360)
    Sprite Size=&Vector2DFromFloat(Value=&Lerp_Float(
      A=&RandomRangeFloat(Minimum=400, Maximum=600)
      B=&RandomRangeFloat(Minimum=2000, Maximum=2200)
      Alpha=$System.Normalized_Phase
     ))
   )
   &DynamicMaterialParameters(
    Index 0 Param 1=1
    Index 0 Param 3=1
    Index 0 Param 4=1
    Index 1 Param 1=1000
    Module.Param0WriteEnabled=true
    Module.Param0WriteEnabled1=true
    Module.Param1WriteEnabled=true
    Module.Param1WriteEnabled1=false
    Module.Param2WriteEnabled=true
    Module.Param2WriteEnabled1=false
    Module.Param3WriteEnabled=true
    Module.Param3WriteEnabled1=false
    Write Parameter Index 1=true
   )
   &CameraOffset(Camera Offset Amount=200)
  ]
  particleUpdateScript [
   &ParticleState()
   &ScaleSpriteSize(Scale Factor=&Vector2DFromFloat(Value=&FloatFromCurve(FloatCurve=DataInterface() {propertyValues {Curve {keys [<at 0 0>, {arriveTangent=-3.1325, leaveTangent=-3.1325, time=0.2, value=1}]}}})))
  ]
  rendererClasses [&NiagaraSpriteRendererProperties]
  renderers [{rendererClass=&NiagaraSpriteRendererProperties}]
 }
 Glow001 {
  emitterUpdateScript [
   &EmitterState(
    Life Cycle Mode=::NewEnumerator1
    Loop Behavior=::NewEnumerator1
    Loop Delay=$System.Anticipation_Delay
    Loop Duration=$User.EffectDuration
    UseLoopDelay=true
   )
   &SpawnRate(
    SpawnRate=&FloatFromCurve(
     FloatCurve=DataInterface() {propertyValues {Curve {keys [<at 0 1 RCIM_Cubic>, {arriveTangent=4.9018, interpMode=RCIM_Cubic, leaveTangent=4.9018, tangentMode=RCTM_User, time=2, value=5}]}}}
     CurveIndex=$Emitter.Age
    )
    Use Spawn Probability=false
   ) {module=SpawnRate001}
  ]
  particleSpawnScript [
   &InitializeParticle(
    Color=<rgba 0 0.6521 2 1>
    Lifetime=0.5
    Module.Write Color=true
    Module.Write Lifetime=true
    Module.Write Mass=false
    Module.Write Material Random=true
    Module.Write Position=true
    Module.Write Scale=false
    Module.Write SpriteRotation=true
    Module.Write SpriteSize=true
    Sprite Rotation=&RandomRangeFloat(Minimum=0, Maximum=360)
    Sprite Size=&Vector2DFromFloat(Value=5000)
   )
   &DynamicMaterialParameters(
    Index 0 Param 1=1
    Index 0 Param 3=1
    Index 0 Param 4=1
    Index 1 Param 1=1000
    Module.Param0WriteEnabled=true
    Module.Param0WriteEnabled1=true
    Module.Param1WriteEnabled=true
    Module.Param1WriteEnabled1=false
    Module.Param2WriteEnabled=true
    Module.Param2WriteEnabled1=false
    Module.Param3WriteEnabled=true
    Module.Param3WriteEnabled1=false
    Write Parameter Index 1=true
   )
   &CameraOffset(Camera Offset Amount=200)
  ]
  particleUpdateScript [
   &ParticleState()
   &ScaleSpriteSize(Scale Factor=&Vector2DFromFloat(Value=&FloatFromCurve(
      FloatCurve=DataInterface() {propertyValues {Curve {keys [<at 0 0 RCIM_Cubic>,<at 0.2 1 RCIM_Cubic>,<at 1 0 RCIM_Cubic>]}}}
      Scale Curve=&Multiply_Float(
       A=&MakeFloatFromInt(Int=&Add_Integer(A=&ReturnParticleIDAttributes(), B=1))
       B=0.25
      )
     )))
  ]
  rendererClasses [&NiagaraSpriteRendererProperties]
  renderers [{rendererClass=&NiagaraSpriteRendererProperties}]
 }
 Opening {
  emitterUpdateScript [
   &EmitterState(Life Cycle Mode=::NewEnumerator1, Loop Behavior=::NewEnumerator1, Loop Duration Mode=::NewEnumerator1, UseLoopDelay=false)
   &SpawnBurst_Instantaneous(Spawn Count=1, Use Spawn Probability=false)
  ]
  particleSpawnScript [
   /Niagara/Modules/Spawn/Initialization/V2/InitializeParticle(
    Color=<rgba 1 1 1 0.2>
    Lifetime=2
    Mesh Scale=&Multiply_VectorByFloat(Vector=<175 175 75>, Float=0.38)
    Mesh Scale Mode=::NewEnumerator4
    UsePositionOffset=true
    Write Material Random=true
   )
   &InitialMeshOrientation(Mesh Axis To Orient=<0 0 -1>, Mesh Orientation Mode=::NewEnumerator2, Orientation Vector=<0 0 1>, Use Rotation Vector=true)
  ]
  particleUpdateScript [
   &ParticleState(Kill Particles When Lifetime Has Elapsed=false, Let Infinitely Lived Particles Die When Emitter Deactivates=true)
   &ScaleMeshSize(Scale Factor=&VectorFromFloat(Value=&Lerp_Float(A=0.2, B=1, Alpha=$System.Normalized_Phase)))
   &DynamicMaterialParameters(
    Index 0 Param 1=&Lerp_Float(A=0, B=0.3, Alpha=$System.Normalized_Phase)
    Module.Param0WriteEnabled=true
    Module.Param1WriteEnabled=true
    Module.Param2WriteEnabled=true
    Module.Param3WriteEnabled=true
   )
   &ScaleColor(Scale Alpha=0.2, Scale RGB=&VectorFromFloat(Value=5), ScaleA=true, ScaleRGB=true)
  ]
  rendererClasses [&NiagaraMeshRendererProperties]
  renderers [{rendererClass=&NiagaraMeshRendererProperties, rendererIndex=0}]
 }
 Pointer {
  emitterUpdateScript [
   &EmitterState(Life Cycle Mode=::NewEnumerator1, Loop Behavior=::NewEnumerator1, Loop Duration Mode=::NewEnumerator1, UseLoopDelay=false)
   &SpawnBurst_Instantaneous(Spawn Count=2, Use Spawn Probability=false, UseLoopCountLimit=false)
   /SetVariables() {isSetParameters=true}
  ]
  particleSpawnScript [&InitializeParticle(
    Color=<rgba 0.25 0.25 0.25 0.25>
    Ribbon Width=2500
    Ribbon Width Mode=::NewEnumerator1
    Sprite Size Mode=::NewEnumerator3
    Uniform Sprite Size=500
    UsePositionOffset=false
    Write Material Random=true
   )]
  particleUpdateScript [
   &ParticleState(Kill Particles When Lifetime Has Elapsed=false, Let Infinitely Lived Particles Die When Emitter Deactivates=true)
   /SetVariables(Particles.Position=&Lerp_Vector(
     A=/Niagara/DynamicInputs/Helpers/SimulationPosition()
     B=&Add_Vector(
      A=&SimulationPosition()
      B=&MakeVector(
       X=0
       Y=0
       Z=&FloatFromCurve(FloatCurve=DataInterface() {propertyValues {Curve {keys [<at 0 0>,<at 1 100000>]}}}, CurveIndex=$Particles.Age)
      )
     )
     Alpha=&ReturnNormalizedExecIndex()
    )) {isSetParameters=true}
   /SetVariables(Particles.RibbonLinkOrder:NiagaraFloat=&ReturnNormalizedExecIndex()) {isSetParameters=true}
   &DynamicMaterialParameters(
    Index 0 Param 1=1
    Index 0 Param 3=&Multiply_Float(A=$Engine.Time, B=0.0088)
    Index 1 Param 1=1
    Module.Param0WriteEnabled=true
    Module.Param0WriteEnabled1=true
    Module.Param1WriteEnabled=true
    Module.Param1WriteEnabled1=true
    Module.Param2WriteEnabled=true
    Module.Param2WriteEnabled1=true
    Module.Param3WriteEnabled=true
    Module.Param3WriteEnabled1=true
    Write Parameter Index 1=true
   )
   &ScaleRibbonWidth(Ribbon Width Scale=&FloatFromCurve(
     FloatCurve=DataInterface() {propertyValues {Curve {keys [{interpMode=RCIM_Cubic, tangentMode=RCTM_User, time=0.5, value=0.9}, {tangentMode=RCTM_User, time=1, value=1}]}}}
     CurveIndex=&ReturnNormalizedExecIndex()
    ))
   &ScaleColor(
    Scale Alpha=&Multiply_Float(
     A=&Lerp_Float(A=1, B=0, Alpha=&ReturnNormalizedExecIndex())
     B=&FloatFromCurve(FloatCurve=DataInterface() {propertyValues {Curve {keys [<at 0.2 0>,<at 1 1>]}}}, CurveIndex=$Emitter.Age)
    )
    Scale RGB=&VectorFromFloat(Value=&FNTimeOfDayScalar(Night Scale=0.5, Morning Scale=0.5, Evening Scale=0.5))
    ScaleA=true
    ScaleRGB=true
   )
   /SetVariables() {isSetParameters=true}
  ]
  rendererClasses [&NiagaraRibbonRendererProperties]
  renderers [{rendererClass=&NiagaraRibbonRendererProperties}]
 }
 QuickCrackVert {
  emitterUpdateScript [
   &EmitterState(Life Cycle Mode=::NewEnumerator1, Loop Behavior=::NewEnumerator1, Loop Delay=4.7333, Loop Duration Mode=::NewEnumerator1, UseLoopDelay=false)
   &SpawnBurst_Instantaneous(Spawn Count=1, Use Spawn Probability=false)
  ]
  particleSpawnScript [
   &InitializeParticle(
    Lifetime=2
    Mesh Scale=&Multiply_VectorByFloat(Vector=<600 600 250>, Float=0.1725)
    Mesh Scale Mode=::NewEnumerator4
    UsePositionOffset=false
    Write Material Random=true
   )
   &InitialMeshOrientation(Mesh Axis To Orient=<0 0 1>, Mesh Orientation Mode=::NewEnumerator2, Rotation=<0 0 0.25>, Use Rotation Vector=true)
  ]
  particleUpdateScript [
   &ParticleState(Kill Particles When Lifetime Has Elapsed=false, Let Infinitely Lived Particles Die When Emitter Deactivates=true)
   &DynamicMaterialParameters(
    Index 0 Param 1=&FloatFromCurve(
     FloatCurve=DataInterface() {propertyValues {
       bOptimizeLUT=false
       bUseLUT=false
       Curve {keys [<at 0 0.01 RCIM_Constant>,<at 0.0033 0.0145 RCIM_Constant>,<at 0.0101 0.017 RCIM_Constant>,<at 0.0177 0.0204 RCIM_Constant>,<at 0.026 0.0245 RCIM_Constant>,<at 0.2163 0.0259 RCIM_Constant>,<at 0.2221 0.0277 RCIM_Constant>,<at 0.2313 0.0308 RCIM_Constant>,<at 0.2467 0.035 RCIM_Constant>,<at 0.4177 0.0348 RCIM_Constant>,<at 0.4227 0.0372 RCIM_Constant>,<at 0.4283 0.0394 RCIM_Constant>,<at 0.4324 0.0426 RCIM_Constant>,<at 0.448 0.0428 RCIM_Constant>,<at 0.452 0.0467 RCIM_Constant>,<at 0.4565 0.0509 RCIM_Constant>,<at 0.6339 0.0509 RCIM_Constant>,<at 0.6564 0.0555 RCIM_Constant>,<at 0.6568 0.0604 RCIM_Constant>,<at 0.6588 0.065 RCIM_Constant>,<at 0.7848 0.065 RCIM_Constant>,<at 0.7881 0.0685 RCIM_Constant>,<at 0.7916 0.0727 RCIM_Constant>,<at 0.8153 0.0728 RCIM_Constant>,<at 0.8182 0.0755 RCIM_Constant>,<at 0.8221 0.0778 RCIM_Constant>,<at 0.8249 0.0803 RCIM_Constant>,<at 0.8518 0.0802 RCIM_Constant>,<at 0.855 0.0835 RCIM_Constant>,<at 0.8571 0.0869 RCIM_Constant>,<at 0.863 0.0885 RCIM_Constant>,<at 0.8651 0.09 RCIM_Constant>,<at 0.9851 0.0901 RCIM_Constant>,<at 0.9918 0.0934 RCIM_Constant>,<at 0.9955 0.0975 RCIM_Constant>,<at 1 0.1 RCIM_Constant>]}
      }}
     CurveIndex=&Add_Float(A=$System.Normalized_Phase, B=-0.1)
     Scale Curve=$System.ScaleCracks
    )
    Index 0 Param 2=&Lerp_Float(A=5, B=10, Alpha=$System.Normalized_Phase)
    Index 0 Param 3=1
    Module.Param0WriteEnabled=true
    Module.Param1WriteEnabled=true
    Module.Param2WriteEnabled=true
    Module.Param3WriteEnabled=true
   ) {module=DynamicMaterialParameters002}
  ]
  rendererClasses [&NiagaraMeshRendererProperties]
  renderers [{rendererClass=&NiagaraMeshRendererProperties, rendererIndex=0}]
 }
 QuickCrackVert001 {
  emitterUpdateScript [
   &EmitterState(Life Cycle Mode=::NewEnumerator1, Loop Behavior=::NewEnumerator1, Loop Delay=6, Loop Duration Mode=::NewEnumerator1, UseLoopDelay=false)
   &SpawnBurst_Instantaneous(Spawn Count=1, Use Spawn Probability=false)
  ]
  particleSpawnScript [
   &InitializeParticle(
    Lifetime=2
    Mesh Scale=&Multiply_VectorByFloat(Vector=<600 600 250>, Float=0.1725)
    Mesh Scale Mode=::NewEnumerator4
    UsePositionOffset=false
    Write Material Random=true
   )
   &InitialMeshOrientation(Mesh Axis To Orient=<0 0 1>, Mesh Orientation Mode=::NewEnumerator2, Rotation=<0 0.5 0>, Use Rotation Vector=true)
  ]
  particleUpdateScript [
   &ParticleState(Kill Particles When Lifetime Has Elapsed=false, Let Infinitely Lived Particles Die When Emitter Deactivates=true)
   &DynamicMaterialParameters(
    Index 0 Param 1=&FloatFromCurve(
     FloatCurve=DataInterface() {propertyValues {
       bOptimizeLUT=false
       bUseLUT=false
       Curve {keys [<at 0 0.01 RCIM_Constant>,<at 0.0033 0.0145 RCIM_Constant>,<at 0.0101 0.017 RCIM_Constant>,<at 0.0177 0.0204 RCIM_Constant>,<at 0.026 0.0245 RCIM_Constant>,<at 0.2163 0.0259 RCIM_Constant>,<at 0.2221 0.0277 RCIM_Constant>,<at 0.2313 0.0308 RCIM_Constant>,<at 0.2467 0.035 RCIM_Constant>,<at 0.4177 0.0348 RCIM_Constant>,<at 0.4227 0.0372 RCIM_Constant>,<at 0.4283 0.0394 RCIM_Constant>,<at 0.4324 0.0426 RCIM_Constant>,<at 0.448 0.0428 RCIM_Constant>,<at 0.452 0.0467 RCIM_Constant>,<at 0.4565 0.0509 RCIM_Constant>,<at 0.6339 0.0509 RCIM_Constant>,<at 0.6564 0.0555 RCIM_Constant>,<at 0.6568 0.0604 RCIM_Constant>,<at 0.6588 0.065 RCIM_Constant>,<at 0.7848 0.065 RCIM_Constant>,<at 0.7881 0.0685 RCIM_Constant>,<at 0.7916 0.0727 RCIM_Constant>,<at 0.8153 0.0728 RCIM_Constant>,<at 0.8182 0.0755 RCIM_Constant>,<at 0.8221 0.0778 RCIM_Constant>,<at 0.8249 0.0803 RCIM_Constant>,<at 0.8518 0.0802 RCIM_Constant>,<at 0.855 0.0835 RCIM_Constant>,<at 0.8571 0.0869 RCIM_Constant>,<at 0.863 0.0885 RCIM_Constant>,<at 0.8651 0.09 RCIM_Constant>,<at 0.9851 0.0901 RCIM_Constant>,<at 0.9918 0.0934 RCIM_Constant>,<at 0.9955 0.0975 RCIM_Constant>,<at 1 0.1 RCIM_Constant>]}
      }}
     CurveIndex=&Add_Float(A=$System.Normalized_Phase, B=-0.15)
     Scale Curve=$System.ScaleCracks
    )
    Index 0 Param 2=&Lerp_Float(A=5, B=10, Alpha=$System.Normalized_Phase)
    Index 0 Param 3=1
    Module.Param0WriteEnabled=true
    Module.Param1WriteEnabled=true
    Module.Param2WriteEnabled=true
    Module.Param3WriteEnabled=true
   ) {module=DynamicMaterialParameters002}
  ]
  rendererClasses [&NiagaraMeshRendererProperties]
  renderers [{rendererClass=&NiagaraMeshRendererProperties, rendererIndex=0}]
 }
 RoundFlareLinger {
  emitterUpdateScript [
   &EmitterState(Life Cycle Mode=::NewEnumerator1, Loop Behavior=::NewEnumerator1, Loop Duration Mode=::NewEnumerator1, UseLoopDelay=true)
   &SpawnBurst_Instantaneous(Use Spawn Probability=false, UseLoopCountLimit=false)
  ]
  particleSpawnScript [
   &InitializeParticle(Lifetime=50, Sprite Size=<8000 8000>, Sprite Size Mode=::NewEnumerator4, UsePositionOffset=false, Write Material Random=true)
   &Color(Color=<rgba 0 0.0652 0.2 1>)
  ]
  particleUpdateScript [
   &ParticleState(Kill Particles When Lifetime Has Elapsed=false, Let Infinitely Lived Particles Die When Emitter Deactivates=true)
   &ScaleSpriteSize(
    Uniform Curve Index=$Particles.Age
    Uniform Curve Scale=&Lerp_Float(A=0.1, B=1, Alpha=$System.Normalized_Phase)
    Uniform Curve Sprite Scale=DataInterface() {propertyValues {Curve {keys [
        {arriveTangent=3.3356, interpMode=RCIM_Cubic, leaveTangent=3.3356, tangentMode=RCTM_User, time=0, value=0}
        {interpMode=RCIM_Cubic, tangentMode=RCTM_User, time=1, value=1}
       ]}}}
   )
   &ScaleColor(
    Scale Alpha=&FloatFromCurve(FloatCurve=DataInterface() {propertyValues {Curve {keys [<at 0 0 RCIM_Cubic>,<at 1 1>]}}}, CurveIndex=$Particles.Age)
    Scale RGB=&VectorFromFloat(Value=&FloatFromCurve(FloatCurve=DataInterface() {propertyValues {Curve {keys [<at 0 1.5 RCIM_Cubic>]}}}))
    ScaleA=true
    ScaleRGB=true
   )
   &DynamicMaterialParameters(
    Index 0 Param 1=&FloatFromCurve(FloatCurve=DataInterface() {propertyValues {Curve {keys [<at 0.5 0.7 RCIM_Cubic>,<at 1 0.1>]}}})
    Index 0 Param 2=1
    Index 0 Param 3=1
    Index 0 Param 4=1
    Module.Param0WriteEnabled=true
    Module.Param1WriteEnabled=true
    Module.Param2WriteEnabled=true
    Module.Param3WriteEnabled=true
   )
   &CameraOffset(Camera Offset Amount=1000)
  ]
  rendererClasses [&NiagaraSpriteRendererProperties]
  renderers [{rendererClass=&NiagaraSpriteRendererProperties}]
 }
}
summary {
 systemName=NS_MidMatch_SkyIsland_Opening
 emitters [
  {emitterName=Crack, rendererClasses [&NiagaraMeshRendererProperties]}
  {emitterName=CrackSmall001, rendererClasses [&NiagaraMeshRendererProperties]}
  {emitterName=Opening, rendererClasses [&NiagaraMeshRendererProperties]}
  {emitterName=Fragments, rendererClasses [&NiagaraMeshRendererProperties]}
  {
   emitterName=FlareLightShockwave001
   rendererClasses [
    &NiagaraSpriteRendererProperties
    &NiagaraLightRendererProperties
    &NiagaraMeshRendererProperties
   ]
  }
  {emitterName=Flash001, rendererClasses [&NiagaraSpriteRendererProperties]}
  {emitterName=QuickCrackVert, rendererClasses [&NiagaraMeshRendererProperties]}
  {emitterName=QuickCrackVert001, rendererClasses [&NiagaraMeshRendererProperties]}
  {emitterName=RoundFlareLinger, rendererClasses [&NiagaraSpriteRendererProperties]}
  {emitterName=Fragments001, rendererClasses [&NiagaraMeshRendererProperties]}
  {emitterName=Forcefield, rendererClasses [&NiagaraMeshRendererProperties]}
  {emitterName=EnergyWave001, rendererClasses [&NiagaraMeshRendererProperties]}
  {emitterName=Glow, rendererClasses [&NiagaraSpriteRendererProperties]}
  {emitterName=Fragments002, rendererClasses [&NiagaraMeshRendererProperties]}
  {emitterName=Glow001, rendererClasses [&NiagaraSpriteRendererProperties]}
  {emitterName=Pointer, rendererClasses [&NiagaraRibbonRendererProperties]}
 ]
 userVariables [
  User.EffectDuration {type=NiagaraFloat, defaultValue=50, description=""}
  User.Height {type=NiagaraFloat, defaultValue=0, description=""}
  User.Lifetime {type=NiagaraFloat, defaultValue=0, description=""}
  User.LoopDurationMain {type=NiagaraFloat, defaultValue=0, description=""}
  User.Normalized_GrowAmount {type=NiagaraFloat, defaultValue=1, description=""}
  User.Radius {type=NiagaraFloat, defaultValue=0.1, description=""}
  User.ZP_Size_Radius {type=NiagaraFloat, defaultValue=0, description=""}
 ]
}
